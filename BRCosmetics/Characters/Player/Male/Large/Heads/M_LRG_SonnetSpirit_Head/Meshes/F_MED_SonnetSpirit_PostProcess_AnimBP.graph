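# Post-process AnimBP for the SonnetSpirit cosmetic head, layering procedural corrections over the incoming pose.
# A Control Rig receives a holding-gun value that weapon changes set to 1, or 0 for pistols.
# The head bone then gets an additive counter-rotation from its scaled neck-relative delta, gated to front end, non-emote or Takeout emote states.
# A constant additive offset also shifts the right aim upper-arm bone.
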AnimBlueprint /BRCosmetics/Characters/Player/Male/Large/Heads/M_LRG_SonnetSpirit_Head/Meshes/F_MED_SonnetSpirit_PostProcess_AnimBP
target_skeleton=/BRCosmetics/Characters/Player/Male/Large/Heads/M_LRG_SonnetSpirit_Head/Meshes/CustomSkeleton/Fortnite_Base_Head_SonnetSpirit_Skeleton
graphs [
 AnimGraph {nodes [
   n1 = Control Rig!(Source=Input Pose(), Alpha, IsHoldingGun:double=Get IsPistolOut() -> (IsPistolOut:double)) -> (Pose) {controlRigAssetReference {blueprintRigClass=&/M_LRG_SonnetSpirit_Head_CtrlRig.M_LRG_SonnetSpirit_Head_CtrlRig_C}}
   n2 = "Transform (Modify) Bone - Bone: aim_upperarm_r"(
    ComponentPose="Transform (Modify) Bone - Bone: head"(
     ComponentPose=Local To Component(LocalPose=@n1)
     Alpha=Get inGame() -> (inGame:double)
     Translation="0,0,0"
     Rotation=Get Head Rot() -> (Head Rot:Rotator)
     Scale="1,1,1"
    ) {rotationMode=BMM_Additive, rotationSpace=BCS_BoneSpace, boneToModify {boneName=head}}
    Alpha=1
    Translation="-7,0,0"
    Rotation="0,0,0"
    Scale="1,1,1"
   ) {
    rotationMode=BMM_Additive
    rotationSpace=BCS_BoneSpace
    translationMode=BMM_Additive
    boneToModify {boneName=aim_upperarm_r}
    translation {x=-7}
   }
   Output Pose(Result=Component To Local(ComponentPose=@n2))
   Get TakeOut Emote() -> (TakeOut Emote:double)
   Get bIsPlayingEmote()
  ]}
 EventGraph {nodes [
   Event BlueprintUpdateAnimation()
   n1 = GetOwnerSkeletalMeshComponent()
   n2 = BreakRotator(InRot=BreakTransform(InTransform=GetDeltaTransformFromRefPose(self=@n1, BoneName=head, BaseName=neck_02)).Rotation)
   n3 = Get z Percentage() -> (z Percentage:double)
   n4 = Get x Percentage() -> (x Percentage:double)
   n5 = Get y Percentage() -> (y Percentage:double)
   > Set Head Rot(Head Rot:Rotator=MakeRotator(
     Roll=float * float!(A=@n2, B=-1, C:double=@n4) -> (ReturnValue) {is=title K2Node_PromotableOperator}
     Pitch=float * float(A=@n2.Pitch, B=-1, C=@n5)
     Yaw=float * float(A=@n2.Yaw, B=-1, C=@n3)
    )) -> (Output_Get:Rotator)
   n6 = NOT Boolean(A=Get bIsPlayingEmote())
   n7 = "Not Equal (Float)"(A=Get TakeOut Emote())
   n8 = Get bIsInFrontEnd()
   > Set inGame(inGame:double="To Float (Boolean)"(InBool=OR Boolean!(A=@n7, B=@n6, C:bool=@n8) -> (ReturnValue))) -> (Output_Get:double)
   n9 = GetAnimInstance(self=GetOwnerSkeletalMeshComponent())
   > Is Valid(InputObject=@n9)
   > Set TakeOut Emote(TakeOut Emote:double=SelectFloat(
     A=1
     B=0
     bPickA=Montage_IsActive!(
      self:Anim Instance Object Reference=@n9
      Montage:Anim Montage Object Reference=/BRCosmetics/Animation/Game/MainPlayer/Emotes/Takeout/Emote_Takeout_CLM_M
     ) -> (ReturnValue:bool)
    )) -> (Output_Get:double)
   Event OnWeaponChanged()
   > Set IsPistolOut(IsPistolOut:double=1) -> (Output_Get:double)
   > n10 = Switch on EFortWeaponCoreAnimation!(execute, Selection:EFortWeaponCoreAnimation Enum=Get WeaponCoreAnim!() -> (WeaponCoreAnim)) -> (
    Melee:Exec
    Pistol:Exec
    Shotgun:Exec
    PaperBlueprint:Exec
    Rifle:Exec
    MeleeOneHand:Exec
    MachinePistol:Exec
    RocketLauncher:Exec
    GrenadeLauncher:Exec
    GoingCommando:Exec
    AssaultRifle:Exec
    TacticalShotgun:Exec
    SniperRifle:Exec
    TrapPlacement:Exec
    ShoulderLauncher:Exec
    AbilityDecoTool:Exec
    Crossbow:Exec
    C4:Exec
    RemoteControl:Exec
    DualWield:Exec
    AR_BullPup:Exec
    AR_ForwardGrip:Exec
    MedPackPaddles:Exec
    SMG_P90:Exec
    AR_DrumGun:Exec
    Consumable_Small:Exec
    Consumable_Large:Exec
    Balloon:Exec
    MountedTurret:Exec
    CreativeTool:Exec
    ExplosiveBow:Exec
    AshtonIndigo:Exec
    AshtonChicago:Exec
    MeleeDualWield:Exec
    Unarmed:Exec
   )
   Set IsPistolOut(execute=@n10.Pistol, IsPistolOut=0)
   TryGetPawnOwner()
  ]}
]
variables [
 Head Rot {type=Rotator}
 inGame {type=number}
 IsPistolOut {type=number}
 TakeOut Emote {type=number}
 x Percentage {type=number, default=0.8}
 y Percentage {type=number, default=0.8}
 z Percentage {type=number, default=1.3}
]
