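# Animation Blueprint for the SonnetSpirit large male head cosmetic, driving flipbook face material parameters from animation curves.
# On customization load it caches a dynamic overlay material, then each update copies left and right brow index curves into it after a zero-length Delay.
# It also compares eye and mouth index curves with stored values, raising the mesh materials' Temporal Responsiveness when any changes, then stores the new values.
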AnimBlueprint /BRCosmetics/Characters/Player/Male/Large/Heads/M_LRG_SonnetSpirit_Head/Meshes/M_LRG_SonnetSpirit_Head_AnimBP
target_skeleton=/BRCosmetics/Characters/Player/Male/Large/Heads/M_LRG_SonnetSpirit_Head/Meshes/CustomSkeleton/Fortnite_Base_Head_SonnetSpirit_Skeleton
graphs [
 EventGraph {nodes [
   n1 = Event BlueprintUpdateAnimation()
   > Parent: BlueprintUpdateAnimation!(execute, DeltaTimeX:float=@n1.DeltaTimeX) -> (then) {is=title K2Node_CallParentFunction}
   > Delay(Duration=0)
   > SetScalarParameterValue(
    self=Get OverlayMaterial() -> (OverlayMaterial:Material Instance Dynamic Object Reference)
    ParameterName=flipbook_face_L_brow_index
    Value=GetCurveValue(CurveName=flipbook_face_L_brow_index)
   )
   > SetScalarParameterValue(self=Get OverlayMaterial(), ParameterName=flipbook_face_R_brow_index, Value=GetCurveValue(CurveName=flipbook_face_R_brow_index))
   n2 = GetPartSkeletalMeshComponent()
   n3 = GetAnimInstance(self=@n2)
   n4 = GetCurveValue(self=@n3, CurveName=flipbook_face_R_eye_index)
   n5 = "Not Equal (Float)"(A=@n4, B=Get RightEyeIndex() -> (RightEyeIndex:float))
   n6 = GetCurveValue(self=@n3, CurveName=flipbook_face_L_eye_index)
   n7 = "Not Equal (Float)"(A=@n6, B=Get LeftEyeIndex() -> (LeftEyeIndex:float))
   n8 = GetCurveValue(self=@n3, CurveName=flipbook_face_mouth_index)
   n9 = "Not Equal (Float)"(A=@n8, B=Get MouthIndex() -> (MouthIndex:float))
   > SetScalarParameterValueOnMaterials!(
    execute
    self:Mesh Component Object Reference=@n2
    ParameterName:Name=Temporal Responsiveness
    ParameterValue:float=SelectFloat(A=1, B=0.5, bPickA=OR Boolean!(A=@n5, B=@n7, C:bool=@n9) -> (ReturnValue))
   ) -> (then)
   > Set RightEyeIndex(RightEyeIndex:float=@n4) -> (Output_Get:float)
   > Set LeftEyeIndex(LeftEyeIndex:float=@n6) -> (Output_Get:float)
   > Set MouthIndex(MouthIndex:float=@n8) -> (Output_Get:float)
   Event OnCharacterCustomizationLoaded()
   > n10 = GetOrCreateDynamicOverlayMaterial(MeshComponent=GetPartSkeletalMeshComponent(), MaterialName=None)
   > Set OverlayMaterial(OverlayMaterial:Material Instance Dynamic Object Reference=@n10) -> (Output_Get:Material Instance Dynamic Object Reference)
  ]}
 NewEventGraph {links [], nodes []}
]
variables [
 IsPistolOut {type=number}
 LeftEyeIndex {type=number}
 MouthIndex {type=number}
 OverlayMaterial {type=MaterialInstanceDynamic}
 RightEyeIndex {type=number}
]
